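# Master material for the HighTower Tomato backpack cosmetic, layering optional skin effects and a glowing counter onto the base character material functions.
# PreFX output passes through static-switched panning emissive and clear coat, masked by channels of SkinFX_Mask, then an emissive Fresnel function.
# MF_counter, tinted and boosted by color and strength parameters, is lerped into emissive by its red channel.
# The modified material attributes then feed the character PostFX function.
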Material /BRCosmetics/Accessories/FORT_Backpacks/Backpack_M_MED_HighTower_Tomato/Materials/M_MED_HighTower_Tomato_Backpack_MASTER
graph {
 properties {
  bUseMaterialAttributes=true
  bUsedWithMeshDeformer=true
  bUsedWithSkeletalMesh=true
  pixelDepthOffsetMode=PDOM_Legacy
  shadingModel=MSM_FromMaterialExpression
  usedShadingModels=Default Lit | Clear Coat | Subsurface Profile
 }
 nodes [
  n1 = MaterialFunctionCall() {materialFunction=&BaseFN_CharacterFunction_PreFX}
  n2 = TextureSampleParameter2D() {parameterName=SkinFX_Mask}
  n3 = StaticSwitchParameter(
   True=MaterialFunctionCall!(
    Alpha=StaticComponentMaskParameter(None=@n2.RGB) {defaultR=true, group=Skin FX, parameterName=PanningEmissive_Channel}
    isVehicle
    Gmap
    UVOffset=StaticComponentMaskParameter(None=@n2.RGB) {defaultB=true, group=Skin FX, parameterName=PanningEmissive_UVOffset_Channel}
    Local Normal
    Local Position=@n1.Pre skinned Local Position
    EmissiveColorOverride
    Material=@n1.Pre skinned Local Position
   ) {materialFunction=/Game/Weapons/WeaponSkins/MaterialFunction/PanningEmissive}.Result
   False=@n1.Result
  ) {group=Skin FX, parameterName=Use Panning Emissive}
  n4 = MaterialFunctionCall!(Alpha=StaticComponentMaskParameter(None=@n2.RGB) {defaultR=true, group=Skin FX, parameterName=Clear Coat Channel}, In=@n3) {materialFunction=/Game/Weapons/WeaponSkins/MaterialFunction/Clearcoat}
  n5 = GetMaterialAttributes(None=MaterialFunctionCall!(In=StaticSwitchParameter(True=@n4.Result, False=@n3) {group=Skin FX, parameterName=Use Clear Coat}) {materialFunction=/Game/Weapons/WeaponSkins/MaterialFunction/MF_EmissiveFresnel}.Result) {attributeGetTypes [#1]}
  n6 = MaterialFunctionCall() {materialFunction=/Game/Gadgets/Assets/VinderTech_Umbrella/Umbrella_BuildABrella/BAB_Counter/Materials/MF_counter}
  n7 = Multiply(
   A=Multiply(A=@n6.Result, B=VectorParameter() {parameterName=CounterColor, defaultValue=<rgba 1 1 1 1>}.RGB)
   B=ScalarParameter() {defaultValue=5, parameterName=CounterEmissiveStrength}
  )
  n8 = ComponentMask(None=@n6.Result) {r=true}
  MaterialFunctionCall(
   AO=@n1.AO
   Distance Blend=@n1.AO
   Fresnel=@n1.AO
   Pre Skinned Local Position=@n1.AO
   Material=SetMaterialAttributes(MaterialAttributes=@n5.MaterialAttributes, Emissive Color=LinearInterpolate(A=@n5.Emissive Color, B=@n7, Alpha=@n8)) {attributeSetTypes [#1]}
  ) {materialFunction=&BaseFN_CharacterFunction_PostFX}
  MaterialFunctionCall() {materialFunction=&MF_EmissiveFresnel}
 ]
}
