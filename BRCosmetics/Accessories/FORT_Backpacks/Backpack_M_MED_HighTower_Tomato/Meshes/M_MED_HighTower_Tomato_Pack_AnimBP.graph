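# Animation Blueprint for the HighTower Tomato backpack, folding twin cannons and panels out while the player is targeting, or on a timed loop in the front end.
# Three hand-posed Modify Bone chains (stowed, intermediate, deployed) mix through two interpolated blends, sequenced by delayed Stage_1 and stage_2 changes.
# Deployed cannon tips pitch up or down from the owner's head-to-pelvis rotation.
# The kill-stat watcher and Look At node are unconnected.
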AnimBlueprint /BRCosmetics/Accessories/FORT_Backpacks/Backpack_M_MED_HighTower_Tomato/Meshes/M_MED_HighTower_Tomato_Pack_AnimBP
target_skeleton=&/M_MED_HighTower_Tomato_Pack_Skeleton
graphs [
 AnimGraph {nodes [
   Look At - Bone: Dyn_Cannon_r_2(Alpha=1) {alphaBoolBlend {blendOption=Linear}, boneToModify {boneName=Dyn_Cannon_r_2}, lookAtTarget=~}
   n1 = "Transform (Modify) Bone - Bone: Dyn_Cannon_Master_r"(ComponentPose=Local To Component(LocalPose=Local Space Ref Pose()), Alpha=1, Translation="0,0,0", Rotation="0,0,0", Scale="1,1,1") {
    rotationMode=BMM_Replace
    rotationSpace=BCS_BoneSpace
    scaleMode=BMM_Replace
    scaleSpace=BCS_BoneSpace
    translationMode=BMM_Replace
    translationSpace=BCS_BoneSpace
    alphaBoolBlend {blendOption=Linear}
    boneToModify {boneName=Dyn_Cannon_Master_r}
   }
   n2 = "Transform (Modify) Bone - Bone: Dyn_Cannon_r_1"(ComponentPose=@n1, Alpha=1, Translation="0,0,0", Rotation="0,0,0", Scale="1,1,1") {
    rotationMode=BMM_Replace
    rotationSpace=BCS_BoneSpace
    scaleMode=BMM_Replace
    scaleSpace=BCS_BoneSpace
    translationMode=BMM_Replace
    translationSpace=BCS_BoneSpace
    alphaBoolBlend {blendOption=Linear}
    boneToModify {boneName=Dyn_Cannon_r_1}
   }
   n3 = "Transform (Modify) Bone - Bone: Dyn_Cannon_r_2"(ComponentPose=@n2, Alpha=1, Translation="0,0,0", Rotation="0,0,0", Scale="0.8,0.8,0.8") {
    rotationMode=BMM_Replace
    rotationSpace=BCS_BoneSpace
    scaleMode=BMM_Replace
    scaleSpace=BCS_BoneSpace
    translationMode=BMM_Replace
    translationSpace=BCS_BoneSpace
    alphaBoolBlend {blendOption=Linear}
    boneToModify {boneName=Dyn_Cannon_r_2}
    scale=<0.8 0.8 0.8>
   }
   n4 = "Transform (Modify) Bone - Bone: Dyn_Panel_r"(ComponentPose=@n3, Alpha=1, Translation="-2.8,0,-1", Rotation="10,0,0", Scale="0.95,0.95,0.95") {
    rotationMode=BMM_Replace
    rotationSpace=BCS_BoneSpace
    scaleMode=BMM_Replace
    scaleSpace=BCS_BoneSpace
    translationMode=BMM_Replace
    translationSpace=BCS_BoneSpace
    alphaBoolBlend {blendOption=Linear}
    boneToModify {boneName=Dyn_Panel_r}
    rotation {pitch=10}
    scale=<0.95 0.95 0.95>
    translation {x=-2.8, z=-1}
   }
   n5 = "Transform (Modify) Bone - Bone: Bottom_Panel_r"(ComponentPose=@n4, Alpha=1, Translation="0,0,1.5", Rotation="5,0,0", Scale="1,1,1") {
    rotationMode=BMM_Replace
    rotationSpace=BCS_BoneSpace
    scaleMode=BMM_Replace
    scaleSpace=BCS_BoneSpace
    translationMode=BMM_Replace
    translationSpace=BCS_BoneSpace
    alphaBoolBlend {blendOption=Linear}
    boneToModify {boneName=Bottom_Panel_r}
    rotation {pitch=5}
    translation {z=1.5}
   }
   n6 = "Transform (Modify) Bone - Bone: Dyn_Cannon_Master_l"(ComponentPose=@n5, Alpha=1, Translation="0,0,0", Rotation="0,0,0", Scale="1,1,1") {
    rotationMode=BMM_Replace
    rotationSpace=BCS_BoneSpace
    scaleMode=BMM_Replace
    scaleSpace=BCS_BoneSpace
    translationMode=BMM_Replace
    translationSpace=BCS_BoneSpace
    alphaBoolBlend {blendOption=Linear}
    boneToModify {boneName=Dyn_Cannon_Master_l}
   }
   n7 = "Transform (Modify) Bone - Bone: Dyn_Cannon_l_1"(ComponentPose=@n6, Alpha=1, Translation="0,0,0", Rotation="0,0,0", Scale="1,1,1") {
    rotationMode=BMM_Replace
    rotationSpace=BCS_BoneSpace
    scaleMode=BMM_Replace
    scaleSpace=BCS_BoneSpace
    translationMode=BMM_Replace
    translationSpace=BCS_BoneSpace
    alphaBoolBlend {blendOption=Linear}
    boneToModify {boneName=Dyn_Cannon_l_1}
   }
   n8 = "Transform (Modify) Bone - Bone: Dyn_Cannon_l_2"(ComponentPose=@n7, Alpha=1, Translation="0,0,0", Rotation="0,0,0", Scale="0.8,0.8,0.8") {
    rotationMode=BMM_Replace
    rotationSpace=BCS_BoneSpace
    scaleMode=BMM_Replace
    scaleSpace=BCS_BoneSpace
    translationMode=BMM_Replace
    translationSpace=BCS_BoneSpace
    alphaBoolBlend {blendOption=Linear}
    boneToModify {boneName=Dyn_Cannon_l_2}
    scale=<0.8 0.8 0.8>
   }
   n9 = "Transform (Modify) Bone - Bone: Dyn_Panel_l"(ComponentPose=@n8, Alpha=1, Translation="2.8,0,1", Rotation="8,0,0", Scale="0.95,0.95,0.95") {
    rotationMode=BMM_Replace
    rotationSpace=BCS_BoneSpace
    scaleMode=BMM_Replace
    scaleSpace=BCS_BoneSpace
    translationMode=BMM_Replace
    translationSpace=BCS_BoneSpace
    alphaBoolBlend {blendOption=Linear}
    boneToModify {boneName=Dyn_Panel_l}
    rotation {pitch=8}
    scale=<0.95 0.95 0.95>
    translation {x=2.8, z=1}
   }
   n10 = "Transform (Modify) Bone - Bone: Bottom_Panel_l"(ComponentPose=@n9, Alpha=1, Translation="0,0,-1.5", Rotation="5,0,0", Scale="1,1,1") {
    rotationMode=BMM_Additive
    rotationSpace=BCS_BoneSpace
    scaleMode=BMM_Replace
    scaleSpace=BCS_BoneSpace
    translationMode=BMM_Replace
    translationSpace=BCS_BoneSpace
    alphaBoolBlend {blendOption=Linear}
    boneToModify {boneName=Bottom_Panel_l}
    rotation {pitch=5}
    translation {z=-1.5}
   }
   n11 = "Transform (Modify) Bone - Bone: Upper_Panel_l"(ComponentPose=@n10, Alpha=1, Translation="0,0,0", Rotation="0,0,0", Scale="1,1,1") {
    rotationMode=BMM_Additive
    rotationSpace=BCS_BoneSpace
    scaleMode=BMM_Replace
    scaleSpace=BCS_BoneSpace
    translationMode=BMM_Replace
    translationSpace=BCS_BoneSpace
    alphaBoolBlend {blendOption=Linear}
    boneToModify {boneName=Upper_Panel_l}
   }
   n12 = "Transform (Modify) Bone - Bone: Upper_Panel_r"(ComponentPose=@n11, Alpha=1, Translation="0,0,0", Rotation="0,0,0", Scale="1,1,1") {
    rotationMode=BMM_Additive
    rotationSpace=BCS_BoneSpace
    scaleMode=BMM_Replace
    scaleSpace=BCS_BoneSpace
    translationMode=BMM_Replace
    translationSpace=BCS_BoneSpace
    alphaBoolBlend {blendOption=Linear}
    boneToModify {boneName=Upper_Panel_r}
   }
   n13 = "Transform (Modify) Bone - Bone: Dyn_Cannon_r_2"(
    ComponentPose=@n12
    Alpha=Get cannon_aim_down_alpha() -> (cannon_aim_down_alpha:double)
    Translation="0,0,0"
    Rotation="-15,20,0"
    Scale="1,1,1"
   ) {
    rotationMode=BMM_Additive
    rotationSpace=BCS_BoneSpace
    alphaBoolBlend {blendOption=Linear}
    boneToModify {boneName=Dyn_Cannon_r_2}
    rotation {pitch=-15, yaw=20}
   }
   n14 = "Transform (Modify) Bone - Bone: Dyn_Cannon_l_2"(ComponentPose=@n13, Alpha=Get cannon_aim_down_alpha(), Translation="0,0,0", Rotation="-15,20,0", Scale="1,1,1") {
    rotationMode=BMM_Additive
    rotationSpace=BCS_BoneSpace
    scaleSpace=BCS_WorldSpace
    alphaBoolBlend {blendOption=Linear}
    boneToModify {boneName=Dyn_Cannon_l_2}
    rotation {pitch=-15, yaw=20}
   }
   n15 = Get cannon_aim_up_alpha() -> (cannon_aim_up_alpha:double)
   n16 = "Transform (Modify) Bone - Bone: Dyn_Cannon_r_2"(ComponentPose=@n14, Alpha=@n15, Translation="0,0,0", Rotation="-15,-30,0", Scale="1,1,1") {
    rotationMode=BMM_Additive
    rotationSpace=BCS_BoneSpace
    alphaBoolBlend {blendOption=Linear}
    boneToModify {boneName=Dyn_Cannon_r_2}
    rotation {pitch=-15, yaw=-30}
   }
   n17 = "Transform (Modify) Bone - Bone: Dyn_Cannon_l_2"(ComponentPose=@n16, Alpha=@n15, Translation="0,0,0", Rotation="-15,-30,0", Scale="1,1,1") {
    rotationMode=BMM_Additive
    rotationSpace=BCS_BoneSpace
    scaleSpace=BCS_WorldSpace
    alphaBoolBlend {blendOption=Linear}
    boneToModify {boneName=Dyn_Cannon_l_2}
    rotation {pitch=-15, yaw=-30}
   }
   n18 = "Transform (Modify) Bone - Bone: Dyn_Cannon_Master_r"(ComponentPose=Local To Component(LocalPose=Local Space Ref Pose()), Alpha=1, Translation="0,0,0", Rotation="0,0,0", Scale="0.75,0.75,0.75") {
    rotationMode=BMM_Replace
    rotationSpace=BCS_BoneSpace
    scaleMode=BMM_Replace
    scaleSpace=BCS_BoneSpace
    translationMode=BMM_Replace
    translationSpace=BCS_BoneSpace
    alphaBoolBlend {blendOption=Linear}
    boneToModify {boneName=Dyn_Cannon_Master_r}
    scale=<0.75 0.75 0.75>
   }
   n19 = "Transform (Modify) Bone - Bone: Dyn_Cannon_r_1"(ComponentPose=@n18, Alpha=1, Translation="0,0,0", Rotation="0,-120,0", Scale="1,1,1") {
    rotationMode=BMM_Replace
    rotationSpace=BCS_BoneSpace
    scaleMode=BMM_Replace
    scaleSpace=BCS_BoneSpace
    translationMode=BMM_Replace
    translationSpace=BCS_BoneSpace
    alphaBoolBlend {blendOption=Linear}
    boneToModify {boneName=Dyn_Cannon_r_1}
    rotation {yaw=-120}
   }
   n20 = "Transform (Modify) Bone - Bone: Dyn_Cannon_r_2"(ComponentPose=@n19, Alpha=1, Translation="0,0,0", Rotation="0,0,0", Scale="0.8,0.8,0.8") {
    rotationMode=BMM_Replace
    rotationSpace=BCS_BoneSpace
    scaleMode=BMM_Replace
    scaleSpace=BCS_BoneSpace
    translationMode=BMM_Replace
    translationSpace=BCS_BoneSpace
    alphaBoolBlend {blendOption=Linear}
    boneToModify {boneName=Dyn_Cannon_r_2}
    scale=<0.8 0.8 0.8>
   }
   n21 = "Transform (Modify) Bone - Bone: Dyn_Panel_r"(ComponentPose=@n20, Alpha=1, Translation="-5,0,-1", Rotation="8,0,0", Scale="0.98,0.98,0.98") {
    rotationMode=BMM_Replace
    rotationSpace=BCS_BoneSpace
    scaleMode=BMM_Replace
    scaleSpace=BCS_BoneSpace
    translationMode=BMM_Replace
    translationSpace=BCS_BoneSpace
    alphaBoolBlend {blendOption=Linear}
    boneToModify {boneName=Dyn_Panel_r}
    rotation {pitch=8}
    scale=<0.98 0.98 0.98>
    translation {x=-5, z=-1}
   }
   n22 = "Transform (Modify) Bone - Bone: Bottom_Panel_r"(ComponentPose=@n21, Alpha=1, Translation="0,0,0", Rotation="0,0,0", Scale="1,1,1") {
    rotationMode=BMM_Replace
    rotationSpace=BCS_BoneSpace
    scaleMode=BMM_Replace
    scaleSpace=BCS_BoneSpace
    translationMode=BMM_Replace
    translationSpace=BCS_BoneSpace
    alphaBoolBlend {blendOption=Linear}
    boneToModify {boneName=Bottom_Panel_r}
   }
   n23 = "Transform (Modify) Bone - Bone: Dyn_Cannon_Master_l"(ComponentPose=@n22, Alpha=1, Translation="0,0,0", Rotation="0,0,0", Scale="0.75,0.75,0.75") {
    rotationMode=BMM_Replace
    rotationSpace=BCS_BoneSpace
    scaleMode=BMM_Replace
    scaleSpace=BCS_BoneSpace
    translationMode=BMM_Replace
    translationSpace=BCS_BoneSpace
    alphaBoolBlend {blendOption=Linear}
    boneToModify {boneName=Dyn_Cannon_Master_l}
    scale=<0.75 0.75 0.75>
   }
   n24 = "Transform (Modify) Bone - Bone: Dyn_Cannon_l_1"(ComponentPose=@n23, Alpha=1, Translation="0,0,0", Rotation="0,-120,0", Scale="1,1,1") {
    rotationMode=BMM_Replace
    rotationSpace=BCS_BoneSpace
    scaleMode=BMM_Replace
    scaleSpace=BCS_BoneSpace
    translationMode=BMM_Replace
    translationSpace=BCS_BoneSpace
    alphaBoolBlend {blendOption=Linear}
    boneToModify {boneName=Dyn_Cannon_l_1}
    rotation {yaw=-120}
   }
   n25 = "Transform (Modify) Bone - Bone: Dyn_Cannon_l_2"(ComponentPose=@n24, Alpha=1, Translation="0,0,0", Rotation="0,0,0", Scale="0.8,0.8,0.8") {
    rotationMode=BMM_Replace
    rotationSpace=BCS_BoneSpace
    scaleMode=BMM_Replace
    scaleSpace=BCS_BoneSpace
    translationMode=BMM_Replace
    translationSpace=BCS_BoneSpace
    alphaBoolBlend {blendOption=Linear}
    boneToModify {boneName=Dyn_Cannon_l_2}
    scale=<0.8 0.8 0.8>
   }
   n26 = "Transform (Modify) Bone - Bone: Dyn_Panel_l"(ComponentPose=@n25, Alpha=1, Translation="5,0,1", Rotation="0.8,0,0", Scale="0.98,0.98,0.98") {
    rotationMode=BMM_Replace
    rotationSpace=BCS_BoneSpace
    scaleMode=BMM_Replace
    scaleSpace=BCS_BoneSpace
    translationMode=BMM_Replace
    translationSpace=BCS_BoneSpace
    alphaBoolBlend {blendOption=Linear}
    boneToModify {boneName=Dyn_Panel_l}
    rotation {pitch=0.8}
    scale=<0.98 0.98 0.98>
    translation {x=5, z=1}
   }
   n27 = "Transform (Modify) Bone - Bone: Bottom_Panel_l"(ComponentPose=@n26, Alpha=1, Translation="0,0,0", Rotation="0,0,0", Scale="1,1,1") {
    rotationMode=BMM_Replace
    rotationSpace=BCS_BoneSpace
    scaleMode=BMM_Replace
    scaleSpace=BCS_BoneSpace
    translationMode=BMM_Replace
    translationSpace=BCS_BoneSpace
    alphaBoolBlend {blendOption=Linear}
    boneToModify {boneName=Bottom_Panel_l}
   }
   n28 = "Transform (Modify) Bone - Bone: Upper_Panel_l"(ComponentPose=@n27, Alpha=1, Translation="0,0,0", Rotation="0,0,0", Scale="1,1,1") {
    rotationMode=BMM_Additive
    rotationSpace=BCS_BoneSpace
    scaleMode=BMM_Replace
    scaleSpace=BCS_BoneSpace
    translationMode=BMM_Replace
    translationSpace=BCS_BoneSpace
    alphaBoolBlend {blendOption=Linear}
    boneToModify {boneName=Upper_Panel_l}
   }
   n29 = "Transform (Modify) Bone - Bone: Upper_Panel_r"(ComponentPose=@n28, Alpha=1, Translation="0,0,0", Rotation="0,0,0", Scale="1,1,1") {
    rotationMode=BMM_Additive
    rotationSpace=BCS_BoneSpace
    scaleMode=BMM_Replace
    scaleSpace=BCS_BoneSpace
    translationMode=BMM_Replace
    translationSpace=BCS_BoneSpace
    alphaBoolBlend {blendOption=Linear}
    boneToModify {boneName=Upper_Panel_r}
   }
   n30 = "Transform (Modify) Bone - Bone: Dyn_Cannon_Master_r"(ComponentPose=Local To Component(LocalPose=Local Space Ref Pose()), Alpha=1, Translation="0,0,0", Rotation="0,0,0", Scale="0.1,0.1,0.1") {
    rotationMode=BMM_Replace
    rotationSpace=BCS_BoneSpace
    scaleMode=BMM_Replace
    scaleSpace=BCS_BoneSpace
    translationMode=BMM_Replace
    translationSpace=BCS_BoneSpace
    alphaBoolBlend {blendOption=Linear}
    boneToModify {boneName=Dyn_Cannon_Master_r}
    scale=<0.1 0.1 0.1>
   }
   n31 = "Transform (Modify) Bone - Bone: Dyn_Cannon_r_1"(ComponentPose=@n30, Alpha=1, Translation="0,0,0", Rotation="0,0,0", Scale="0.1,0.1,0.1") {
    rotationMode=BMM_Replace
    rotationSpace=BCS_BoneSpace
    scaleMode=BMM_Replace
    scaleSpace=BCS_BoneSpace
    translationMode=BMM_Replace
    translationSpace=BCS_BoneSpace
    alphaBoolBlend {blendOption=Linear}
    boneToModify {boneName=Dyn_Cannon_r_1}
    scale=<0.1 0.1 0.1>
   }
   n32 = "Transform (Modify) Bone - Bone: Dyn_Cannon_r_2"(ComponentPose=@n31, Alpha=1, Translation="0,0,0", Rotation="0,0,0", Scale="0.1,0.1,0.1") {
    rotationMode=BMM_Replace
    rotationSpace=BCS_BoneSpace
    scaleMode=BMM_Replace
    scaleSpace=BCS_BoneSpace
    translationMode=BMM_Replace
    translationSpace=BCS_BoneSpace
    alphaBoolBlend {blendOption=Linear}
    boneToModify {boneName=Dyn_Cannon_r_2}
    scale=<0.1 0.1 0.1>
   }
   n33 = "Transform (Modify) Bone - Bone: Dyn_Panel_r"(ComponentPose=@n32, Alpha=1, Translation="0,0,0", Rotation="0,0,0", Scale="1,1,1") {
    rotationMode=BMM_Replace
    rotationSpace=BCS_BoneSpace
    scaleMode=BMM_Replace
    scaleSpace=BCS_BoneSpace
    translationMode=BMM_Replace
    translationSpace=BCS_BoneSpace
    alphaBoolBlend {blendOption=Linear}
    boneToModify {boneName=Dyn_Panel_r}
   }
   n34 = "Transform (Modify) Bone - Bone: Bottom_Panel_r"(ComponentPose=@n33, Alpha=1, Translation="0,0,-1", Rotation="0,0,0", Scale="1,1,1") {
    rotationMode=BMM_Replace
    rotationSpace=BCS_BoneSpace
    scaleMode=BMM_Replace
    scaleSpace=BCS_BoneSpace
    translationMode=BMM_Replace
    translationSpace=BCS_BoneSpace
    alphaBoolBlend {blendOption=Linear}
    boneToModify {boneName=Bottom_Panel_r}
    translation {z=-1}
   }
   n35 = "Transform (Modify) Bone - Bone: Dyn_Cannon_Master_l"(ComponentPose=@n34, Alpha=1, Translation="0,0,0", Rotation="0,0,0", Scale="0.1,0.1,0.1") {
    rotationMode=BMM_Replace
    rotationSpace=BCS_BoneSpace
    scaleMode=BMM_Replace
    scaleSpace=BCS_BoneSpace
    translationMode=BMM_Replace
    translationSpace=BCS_BoneSpace
    alphaBoolBlend {blendOption=Linear}
    boneToModify {boneName=Dyn_Cannon_Master_l}
    scale=<0.1 0.1 0.1>
   }
   n36 = "Transform (Modify) Bone - Bone: Dyn_Cannon_l_1"(ComponentPose=@n35, Alpha=1, Translation="0,0,0", Rotation="0,0,0", Scale="0.1,0.1,0.1") {
    rotationMode=BMM_Replace
    rotationSpace=BCS_BoneSpace
    scaleMode=BMM_Replace
    scaleSpace=BCS_BoneSpace
    translationMode=BMM_Replace
    translationSpace=BCS_BoneSpace
    alphaBoolBlend {blendOption=Linear}
    boneToModify {boneName=Dyn_Cannon_l_1}
    scale=<0.1 0.1 0.1>
   }
   n37 = "Transform (Modify) Bone - Bone: Dyn_Cannon_l_2"(ComponentPose=@n36, Alpha=1, Translation="0,0,0", Rotation="0,0,0", Scale="0.1,0.1,0.1") {
    rotationMode=BMM_Replace
    rotationSpace=BCS_BoneSpace
    scaleMode=BMM_Replace
    scaleSpace=BCS_BoneSpace
    translationMode=BMM_Replace
    translationSpace=BCS_BoneSpace
    alphaBoolBlend {blendOption=Linear}
    boneToModify {boneName=Dyn_Cannon_l_2}
    scale=<0.1 0.1 0.1>
   }
   n38 = "Transform (Modify) Bone - Bone: Dyn_Panel_l"(ComponentPose=@n37, Alpha=1, Translation="0,0,0", Rotation="0,0,0", Scale="1,1,1") {
    rotationMode=BMM_Replace
    rotationSpace=BCS_BoneSpace
    scaleMode=BMM_Replace
    scaleSpace=BCS_BoneSpace
    translationMode=BMM_Replace
    translationSpace=BCS_BoneSpace
    alphaBoolBlend {blendOption=Linear}
    boneToModify {boneName=Dyn_Panel_l}
   }
   n39 = "Transform (Modify) Bone - Bone: Bottom_Panel_l"(ComponentPose=@n38, Alpha=1, Translation="0,0,1", Rotation="0,0,0", Scale="1,1,1") {
    rotationMode=BMM_Additive
    rotationSpace=BCS_BoneSpace
    scaleMode=BMM_Replace
    scaleSpace=BCS_BoneSpace
    translationMode=BMM_Replace
    translationSpace=BCS_BoneSpace
    alphaBoolBlend {blendOption=Linear}
    boneToModify {boneName=Bottom_Panel_l}
    translation {z=1}
   }
   n40 = "Transform (Modify) Bone - Bone: Upper_Panel_l"(ComponentPose=@n39, Alpha=1, Translation="-1,-1,-1", Rotation="0,0,0", Scale="1,1,1") {
    rotationMode=BMM_Additive
    rotationSpace=BCS_BoneSpace
    scaleMode=BMM_Replace
    scaleSpace=BCS_BoneSpace
    translationMode=BMM_Replace
    translationSpace=BCS_BoneSpace
    alphaBoolBlend {blendOption=Linear}
    boneToModify {boneName=Upper_Panel_l}
    translation=<-1 -1 -1>
   }
   n41 = "Transform (Modify) Bone - Bone: Upper_Panel_r"(ComponentPose=@n40, Alpha=1, Translation="1,-1,-1", Rotation="0,0,0", Scale="1,1,1") {
    rotationMode=BMM_Additive
    rotationSpace=BCS_BoneSpace
    scaleMode=BMM_Replace
    scaleSpace=BCS_BoneSpace
    translationMode=BMM_Replace
    translationSpace=BCS_BoneSpace
    alphaBoolBlend {blendOption=Linear}
    boneToModify {boneName=Upper_Panel_r}
    translation=<1 -1 -1>
   }
   n42 = Two Way Blend(A=Component To Local(ComponentPose=@n41), B=Component To Local(ComponentPose=@n29), Alpha=Get Stage_1() -> (Stage_1:double)) {blendNode {alphaBoolBlend {blendOption=Linear}, alphaScaleBiasClamp {bInterpResult=true, interpSpeedDecreasing=7, interpSpeedIncreasing=13}}}
   n43 = Two Way Blend(A=@n42, B=Component To Local(ComponentPose=@n17), Alpha=Get stage_2() -> (stage_2:double)) {blendNode {alphaBoolBlend {blendOption=Linear}, alphaScaleBiasClamp {bInterpResult=true, interpSpeedIncreasing=13}}}
   Output Pose(Result=Component To Local(ComponentPose=Local To Component(LocalPose=@n43))) {layerGroup=None}
  ]}
 EventGraph {nodes [
   n1 = WatchCosmeticStat!(
    execute:Exec
    Pawn:Fort Player Pawn Object Reference=Get Pawn() -> (Pawn:Fort Player Pawn Object Reference)
    StatTag:Fort Stat Manager Tag Structure=(TagName="StatManager.AthenaKills")
    DefaultValue:int=0
    bBroadcastInitialValue:bool=true
    InitialValueBroadcastDelay:float=-1
    ChangeValueBroadcastDelay:float=-1
    bConsumeWaitingNotifications:bool=true
   ) -> (then:Exec, OnStatChanged:Exec, StatName:Name, StatValue:int) {is=title K2Node_AsyncAction}
   Set Did Kill(execute=@n1.OnStatChanged, Did Kill:bool=false) -> (Output_Get:bool)
   > n2 = Set Kill Count(Kill Count:double="To Float (Integer)"(InInt=@n1.StatValue)) -> (Output_Get:double)
   > Branch(Condition="float > float"(A=@n2, B=Get Last Kill Count() -> (Last Kill Count:double)) {is=title K2Node_CallFunction})
   > Set Active_Kill!(execute, Active_Kill:bool=true, self:M MED Target Practice Pack Skeleton Anim BP Object Reference) -> (then, Output_Get:bool)
   > Set Last Kill Count(Last Kill Count:double=@n2) -> (Output_Get:double)
   Event BlueprintInitializeAnimation()
   > n3 = Cast To FortPlayerPawn(Object=TryGetFortPawn())
   > Set Skeletal Mesh(Skeletal Mesh:Skeletal Mesh Component Object Reference=GetSkeletalMeshForPartType(self=@n3)) -> (Output_Get:Skeletal Mesh Component Object Reference)
   Event BlueprintUpdateAnimation()
   > n4 = Branch(Condition=Get bIsInFrontEnd())
   > n5 = Sequence()
   > Do Once(Reset=@n6.then)
   > Set BackPack_Closed(BackPack_Closed:double=0) -> (Output_Get:double)
   > Delay(Duration=0.2)
   > Set Stage_1(Stage_1:double=1) -> (Output_Get:double)
   > Delay(Duration=0.2)
   > Set stage_2(stage_2:double=1) -> (Output_Get:double)
   > Delay(Duration=0.5)
   > Set Stage_1(Stage_1=0)
   > Delay(Duration=5)
   > Set Stage_1(Stage_1=1)
   > Delay(Duration=0.2)
   > Set stage_2(stage_2=0)
   > Delay(Duration=0.2)
   > Set Stage_1(Stage_1=0)
   n6 = Delay(execute=@n5.then_1, Duration=8.5)
   n7 = GetOwnerSkeletalMeshComponent()
   n8 = BreakRotator(InRot=BreakTransform(InTransform=GetDeltaTransformFromRefPose(self=@n7, BoneName=head, BaseName=pelvis)).Rotation)
   Set cannon_aim_up_alpha(execute=@n4.else, cannon_aim_up_alpha:double=MapRangeClamped(Value=@n8.Yaw, InRangeA=10, InRangeB=-60)) -> (Output_Get:double)
   > Set cannon_aim_down_alpha(cannon_aim_down_alpha:double=MapRangeClamped(Value=@n8.Yaw, InRangeA=10, InRangeB=40)) -> (Output_Get:double)
   > n9 = Branch(Condition=Get bIsTargeting())
   > Do Once(Reset=@n11.then)
   > Set Stage_1(Stage_1=1)
   > Delay(Duration=0.2)
   > Set stage_2(stage_2=1)
   > Delay(Duration=0.5)
   > n10 = Set Stage_1(Stage_1=0)
   Do Once(execute=@n9.else, Reset=@n10.then, Start Closed=true)
   > Set Stage_1(Stage_1=1)
   > Delay(Duration=0.2)
   > Set stage_2(stage_2=0)
   > Delay(Duration=0.3)
   > n11 = Set Stage_1(Stage_1=0)
   TryGetPawnOwner()
   BreakRotator(InRot=GetBaseAimRotation!(self:Pawn Object Reference=@n3) -> (ReturnValue:Rotator))
   n12 = Get Skeletal Mesh() -> (Skeletal Mesh:Skeletal Mesh Component Object Reference)
   GetAlphaFromReferenceTransformDelta(
    SkeletalMeshComponent=@n12
    BoneName=head
    BaseName=spine_05
    DeltaComponentType=QuaternionTwist
    InputRange=(X=-8,Y=-10)
    OutputRange=(X=-10,Y=-5)
    TwistAxis="1,0,0"
   )
   GetAlphaFromReferenceTransformDelta(
    SkeletalMeshComponent=@n12
    BoneName=head
    BaseName=spine_05
    DeltaComponentType=QuaternionTwist
    InputRange=(X=30,Y=-30)
    OutputRange=(X=40,Y=-20)
    TwistAxis="1,0,0"
   )
   Get Pawn()
  ]}
]
variables [
 Active Kill {type=boolean}
 BackPack_Closed {type=number}
 Cannon_Aim {type=Rotator}
 cannon_aim_down_alpha {type=number}
 cannon_aim_up_alpha {type=number}
 Did Kill {type=boolean}
 Gun up and down {type=number}
 Gun_Twist {type=number}
 Kill Count {type=number}
 Last Kill Count {type=number}
 Pawn {type=FortPlayerPawn}
 Skeletal Mesh {type=SkeletalMeshComponent}
 Stage_1 {type=number}
 stage_2 {type=number}
]
